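; Backpack minigun weapon Blueprint that adds spin, barrel-heat and overheat effects to the gun.
; Charge up and down toggle a particle and swap spin sounds, with a timeline scaling volume.
; Firing starts a repeating timer mapping heat ratio to muzzle glow opacity and a smoothed audio parameter, self-clearing at zero heat after a half-second guard.
; Overheating triggers Niagara, sound and a leftover debug print; equip and unequip reset glow and audio.
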(event EventOnAmmoCountChanged
  (|CallParentFunction)
  (Variables|Default|SetCurrentAmmo (FortWeapon|GetLocalMagazineAmmoCount)))

(event Custom|Start%20Spin%20Up%20Audio
  ;[1 1 1] Handle volume scaling on the spin up and spin down audio cues
  (Audio|Components|Audio|FadeOut (Variables|Default|GetSpinDownSound) 0.25 0.0 "Linear" :!class /Script/Engine.AudioComponent)
  (bind _output_get (Variables|Default|SetCurrentSpinAudioComponent (Variables|Default|GetSpinUpSound)))
  (if (not (Variables|Default|GetbPlayedSpinUpAudio))
    (Audio|Components|Audio|Play _output_get 0.0)
    (Audio|Components|Audio|SetVolumeMultiplier _output_get 0.05 :!class /Script/Engine.AudioComponent)
    (Variables|Default|SetbPlayedSpinUpAudio true))
  (bind (_direction _volume) (Timeline|SpinAudioVolumeTimeline Play)
    (:Update
      (Variables|Default|SetSpinVolumeMultiplier _volume)
      (Audio|Components|Audio|SetVolumeMultiplier (Variables|Default|GetCurrentSpinAudioComponent) (Variables|Default|GetSpinVolumeMultiplier) :!class /Script/Engine.AudioComponent))))

(event Overheating|EventOnStartOverheated
  ;[1 1 1] Add overheat FX triggers here!
  (|CallParentFunction)
  (Development|PrintString "OVERHEATING")
  (Components|Activation|Activate (Variables|Default|GetNS_Minigun_Overheated) false)
  (Audio|Components|Audio|Play (Variables|Default|GetOnOverheatSound) 0.0))

(event Custom|Start%20Spin%20Down%20Audio
  ;[1 1 1] Handle volume scaling on the spin up and spin down audio cues
  (Audio|Components|Audio|FadeOut (Variables|Default|GetSpinUpSound) 0.25 0.0 "Linear" :!class /Script/Engine.AudioComponent)
  (bind _output_get (Variables|Default|SetCurrentSpinAudioComponent (Variables|Default|GetSpinDownSound)))
  (if (not (Variables|Default|GetbPlayedSpinDownAudio))
    (Audio|Components|Audio|Play _output_get 0.0)
    (Audio|Components|Audio|SetVolumeMultiplier _output_get 0.05 :!class /Script/Engine.AudioComponent)
    (Variables|Default|SetbPlayedSpinDownAudio true))
  (Timeline|SpinAudioVolumeTimeline Reverse))

(event EventOnChargeUp
  ;[1 1 1] OnStartCharge - Sound
  (|CallParentFunction)
  (Components|Activation|Activate (Variables|Default|GetP_MinigunChargeUp) false)
  (CallFunction|StartSpinUpAudio)
  (Variables|Default|SetbPlayedSpinDownAudio false))

(event EventOnChargeDown
  ;[1 1 1] OnEndCharge - Sound
  (|CallParentFunction)
  (Utilities|FlowControl|Sequence
    (:then_0
      (CallFunction|StartSpinDownAudio)
      (Variables|Default|SetbPlayedSpinUpAudio false))
    (:then_1
      (Components|Activation|Deactivate (Variables|Default|GetP_MinigunChargeUp)))))

(event EventOnPlayWeaponFireFX
  (params (bPersistentFire :type Boolean) (bSecondaryFire :type Boolean))
  ;[1 1 1] OnPlayWeaponFireFX
  (|CallParentFunction bPersistentFire false)
  (Audio|Components|Audio|Play (Variables|Default|GetOverheatLoopSound) 0.0)
  (bind _returnvalue (Utilities|Time|SetTimerbyEvent (EventDispatchers|CreateEvent :!event Custom|BarrelOverheatFXandSound) 0.0333 true))
  (Variables|Default|SetGlowTimer _returnvalue)
  ;[1 1 1] First Half second do not allow timer to end. This prevents it from turning off glow timer before it had time to spin up
  (Variables|Default|SetHitHeatThreshold false)
  (Utilities|FlowControl|RetriggerableDelay 0.5)
  (Variables|Default|SetHitHeatThreshold true))

(event Custom|BarrelOverheatFXandSound
  (bind _returnvalue (CallFunction|CurrentScalarValue))
  (bind _returnvalue_2 (Math|Float|Lerp 0.0 1.0 _returnvalue))
  ;[1 1 1] BarrelGlow - Handles ramping overheat FX and sound.  Timer shuts off when there is no heat in the minigun
  (if (> (CallFunction|CurrentScalarValue) 0.0)
    (Networking|CanExecuteCosmeticEvents
      (:True
        (Rendering|Material|SetScalarParameterValue (Variables|Default|GetMuzzleGlowMID) "Opacity" _returnvalue_2 :!class /Script/Engine.MaterialInstanceDynamic)
        (Variables|Default|SetOverheatSmoothed (Math|Interpolation|FInterpTo (Variables|Default|GetOverheatSmoothed) _returnvalue_2 (Utilities|Time|GetWorldDeltaSeconds) 1e+01))
        (Audio|Parameter|SetFloatParameter (Variables|Default|GetOverheatLoopSound) "Overheat" (Math|Float|MapRangeClamped (Variables|Default|GetOverheatSmoothed) 0.08 1.0 0.0 1.0) :!class /Script/AudioExtensions.AudioParameterControllerInterface)))
    (elif (Variables|Default|GetHitHeatThreshold)
      (Utilities|Time|ClearandInvalidateTimerbyHandle (Variables|Default|GetGlowTimer))
      (if (Audio|Components|Audio|IsPlaying (Variables|Default|GetOverheatLoopSound))
        (Audio|Components|Audio|FadeOut (Variables|Default|GetOverheatLoopSound) 1.0 0.0 "Linear" :!class /Script/Engine.AudioComponent)))))

(event FortWeapon|EventOnEquippedAndAttached
  ;[1 1 1] OnEquip - Reset MuzzleGlowMID Opacity
  (|CallParentFunction)
  (Networking|CanExecuteCosmeticEvents
    (:True
      (Rendering|Material|SetScalarParameterValue (Variables|Default|GetMuzzleGlowMID) "Opacity" 0.0 :!class /Script/Engine.MaterialInstanceDynamic)
      (Audio|Parameter|SetFloatParameter (Variables|Default|GetOverheatLoopSound) "Overheat" 0.0 :!class /Script/AudioExtensions.AudioParameterControllerInterface))))

(event FortWeapon|EventOnLocalUnEquip
  ;[1 1 1] Clear Timer Reset MuzzleGlowMID Opacity/Sound.  Keeps a phantom sound from playing.
  (|CallParentFunction)
  (Utilities|Time|ClearandInvalidateTimerbyHandle (Variables|Default|GetGlowTimer))
  (Networking|CanExecuteCosmeticEvents
    (:True
      (Rendering|Material|SetScalarParameterValue (Variables|Default|GetMuzzleGlowMID) "Opacity" 0.0 :!class /Script/Engine.MaterialInstanceDynamic)
      (Audio|Parameter|SetFloatParameter (Variables|Default|GetOverheatLoopSound) "Overheat" 0.0 :!class /Script/AudioExtensions.AudioParameterControllerInterface))))

(event EventOnInitCosmeticAlterations
  (params (CosmeticMod :type Fort%20Cosmetic%20Modification%20Structure)))

(fn CurrentScalarValue
  (return (/ (Overheat|GetCurrentOverheatValue) (Overheat|GetOverheatingMaxValue))))

(fn ConstructionScript
  (Networking|CanExecuteCosmeticEvents
    (:True
      (bind _returnvalue (Rendering|Material|CreateDynamicMaterialInstance (Variables|Default|GetBarrelGlowFX) 0 0 "None" :!class /Script/Engine.PrimitiveComponent))
      (Variables|Default|SetMuzzleGlowMID _returnvalue))
    (:False
      _))
  (|CallParentFunction))
